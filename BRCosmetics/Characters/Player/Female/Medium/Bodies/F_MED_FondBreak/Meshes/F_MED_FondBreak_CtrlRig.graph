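# Control Rig for the FondBreak female medium body cosmetic, adding corrective deformation to its outfit bones.
# After the shared DeformRig function, Deform_Boots blends boot support bones from remapped foot yaw deltas.
# Deform_Constraints then pins buckles, bag and bow bases, and a button with weighted multi-bone parent constraints.
# Deform_Body layers spherical pose readers on thighs and hands to offset bags and bow ends, smoothing one right-thigh reader with AlphaInterp.
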ControlRigBlueprint /BRCosmetics/Characters/Player/Female/Medium/Bodies/F_MED_FondBreak/Meshes/F_MED_FondBreak_CtrlRig
graphs [
 New Function_ContainedGraph {nodes [
   RigVMFunctionEntryNode()
   > Get Delta Ref = RigVMFunctionReferenceNode(Item=(Type=Bone,Name="foot_r"))
   > &RigUnit_ModifyTransforms(
    ItemToModify=((Item=(Type=Bone,Name="foot_support_r"),Transform=(Rotation=(X=0,Y=0,Z=-0.1263,W=0.992),Translation=(X=-0.4748,Y=3.1047,Z=0),Scale3D=(X=1,Y=1,Z=1))))
    WeightMinimum=0
    Weight=&RigVMFunction_MathDoubleRemap(SourceMaximum=50, SourceMinimum=25, TargetMaximum=1, TargetMinimum=0, Value=@"Get Delta Ref".Result.Rotation.Yaw).Result
   )
   > Get Delta Ref_1 = RigVMFunctionReferenceNode(Item=(Type=Bone,Name="foot_l"))
   > &RigUnit_ModifyTransforms(
    ItemToModify=((Item=(Type=Bone,Name="foot_support_l"),Transform=(Rotation=(X=0,Y=0,Z=0.2983,W=-0.9545),Translation=(X=-3.4817,Y=-8.4624,Z=1.79),Scale3D=(X=1,Y=1,Z=1))))
    WeightMinimum=0
    Weight=&RigVMFunction_MathDoubleRemap(SourceMaximum=50, SourceMinimum=25, TargetMaximum=1, TargetMinimum=0, Value=@"Get Delta Ref_1".Result.Rotation.Yaw).Result
   )
   > RigVMFunctionReturnNode()
  ]}
 New Function_ContainedGraph {nodes [
   RigVMFunctionEntryNode()
   > &RigUnit_ParentConstraint(
    AdvancedSettings=(RotationOrderForFilter=XZY)
    Child=(Type=Bone,Name="deform_buckle_front")
    Parents="((Item=(Type=Bone,Name=\"thigh_r\"),Weight=0.0021),(Item=(Type=Bone,Name=\"thigh_l\"),Weight=0.0022),(Item=(Type=Bone,Name=\"spine_02\"),Weight=0.0116),(Item=(Type=Bone,Name=\"thigh_twist_01_r\"),Weight=0.0625),(Item=(Type=Bone,Name=\"thigh_twist_01_l\"),Weight=0.06 ...(truncated)"
    Weight=1
   )
   > &RigUnit_ParentConstraint(
    AdvancedSettings=(RotationOrderForFilter=XZY)
    Child=(Type=Bone,Name="deform_buckle_back")
    Parents="((Item=(Type=Bone,Name=\"spine_03\"),Weight=0.0003),(Item=(Type=Bone,Name=\"thigh_twist_01_r\"),Weight=0.0006),(Item=(Type=Bone,Name=\"thigh_twist_01_l\"),Weight=0.0006),(Item=(Type=Bone,Name=\"spine_01\"),Weight=0.0071),(Item=(Type=Bone,Name=\"spine_02\"),Weight=0. ...(truncated)"
    Weight=1
   )
   > &RigUnit_ParentConstraint(
    AdvancedSettings=(RotationOrderForFilter=XZY)
    Child=(Type=Bone,Name="dyn_bag_frnt_l_base")
    Parents="((Item=(Type=Bone,Name=\"spine_02\"),Weight=0.0035),(Item=(Type=Bone,Name=\"thigh_l\"),Weight=0.006),(Item=(Type=Bone,Name=\"thigh_twist_01_r\"),Weight=0.0084),(Item=(Type=Bone,Name=\"deform_groin_r\"),Weight=0.0088),(Item=(Type=Bone,Name=\"deform_glute_l\"),Weight= ...(truncated)"
    Weight=1
   )
   > &RigUnit_ParentConstraint(
    AdvancedSettings=(RotationOrderForFilter=XZY)
    Child=(Type=Bone,Name="dyn_bag_side_l_base")
    Parents="((Item=(Type=Bone,Name=\"deform_glute_r\"),Weight=0.0002),(Item=(Type=Bone,Name=\"spine_02\"),Weight=0.0019),(Item=(Type=Bone,Name=\"deform_groin_r\"),Weight=0.002),(Item=(Type=Bone,Name=\"thigh_twist_01_r\"),Weight=0.0023),(Item=(Type=Bone,Name=\"thigh_l\"),Weight= ...(truncated)"
    Weight=1
   )
   > &RigUnit_ParentConstraint(
    AdvancedSettings=(RotationOrderForFilter=XZY)
    Child=(Type=Bone,Name="dyn_bag_frnt_r_base")
    Parents="((Item=(Type=Bone,Name=\"spine_02\"),Weight=0.0059),(Item=(Type=Bone,Name=\"thigh_r\"),Weight=0.0082),(Item=(Type=Bone,Name=\"deform_glute_r\"),Weight=0.0156),(Item=(Type=Bone,Name=\"thigh_twist_01_l\"),Weight=0.0173),(Item=(Type=Bone,Name=\"deform_groin_l\"),Weight ...(truncated)"
    Weight=1
   )
   > &RigUnit_ParentConstraint(
    AdvancedSettings=(RotationOrderForFilter=XZY)
    Child=(Type=Bone,Name="dyn_bag_bk_r_base")
    Parents="((Item=(Type=Bone,Name=\"spine_01\"),Weight=0.0001),(Item=(Type=Bone,Name=\"deform_glute_l\"),Weight=0.0005),(Item=(Type=Bone,Name=\"deform_groin_l\"),Weight=0.0009),(Item=(Type=Bone,Name=\"thigh_twist_01_l\"),Weight=0.0013),(Item=(Type=Bone,Name=\"spine_02\"),Weigh ...(truncated)"
    Weight=1
   )
   > &RigUnit_ParentConstraint(
    AdvancedSettings=(RotationOrderForFilter=XZY)
    Child=(Type=Bone,Name="dyn_bag_side_r_base")
    Parents="((Item=(Type=Bone,Name=\"spine_02\"),Weight=0.0029),(Item=(Type=Bone,Name=\"thigh_r\"),Weight=0.0055),(Item=(Type=Bone,Name=\"deform_groin_l\"),Weight=0.0059),(Item=(Type=Bone,Name=\"thigh_twist_01_l\"),Weight=0.0059),(Item=(Type=Bone,Name=\"deform_glute_r\"),Weight ...(truncated)"
    Weight=1
   )
   > &RigUnit_ParentConstraint(
    AdvancedSettings=(RotationOrderForFilter=XZY)
    Child=(Type=Bone,Name="dyn_hand_bow_l_base")
    Parents="((Item=(Type=Bone,Name=\"dyn_hand_bow_l_base\"),Weight=0.0037),(Item=(Type=Bone,Name=\"lowerarm_twist_02_l\"),Weight=0.0149),(Item=(Type=Bone,Name=\"lowerarm_l\"),Weight=0.024),(Item=(Type=Bone,Name=\"thumb_01_l\"),Weight=0.025),(Item=(Type=Bone,Name=\"deform_wrist ...(truncated)"
    Weight=1
   )
   > &RigUnit_ParentConstraint(
    AdvancedSettings=(RotationOrderForFilter=XZY)
    Child=(Type=Bone,Name="dyn_hand_bow_r_base")
    Parents="((Item=(Type=Bone,Name=\"pinky_metacarpal_r\"),Weight=0.0024),(Item=(Type=Bone,Name=\"lowerarm_twist_02_r\"),Weight=0.013),(Item=(Type=Bone,Name=\"lowerarm_r\"),Weight=0.0208),(Item=(Type=Bone,Name=\"thumb_01_r\"),Weight=0.0229),(Item=(Type=Bone,Name=\"deform_wrist ...(truncated)"
    Weight=1
   )
   > &RigUnit_ParentConstraint(
    AdvancedSettings=(RotationOrderForFilter=XZY)
    Child=(Type=Bone,Name="dyn_foot_bow_l_base")
    Parents=((Item=(Type=Bone,Name="foot_l"),Weight=0.004),(Item=(Type=Bone,Name="calf_twist_01_l"),Weight=0.996))
    Weight=1
   )
   > &RigUnit_ParentConstraint(
    AdvancedSettings=(RotationOrderForFilter=XZY)
    Child=(Type=Bone,Name="dyn_foot_bow_r_base")
    Parents=((Item=(Type=Bone,Name="foot_r"),Weight=0.004),(Item=(Type=Bone,Name="calf_twist_01_r"),Weight=0.996))
    Weight=1
   )
   > &RigUnit_ParentConstraint(
    AdvancedSettings=(RotationOrderForFilter=XZY)
    Child=(Type=Bone,Name="deform_button")
    Parents="((Item=(Type=Bone,Name=\"deform_pec_r\"),Weight=0.0137),(Item=(Type=Bone,Name=\"deform_upperarm_up_r\"),Weight=0.025),(Item=(Type=Bone,Name=\"deform_upperarm_fr_r\"),Weight=0.0421),(Item=(Type=Bone,Name=\"upperarm_twist_01_r\"),Weight=0.0538),(Item=(Type=Bone,Name ...(truncated)"
    Weight=1
   )
   > RigVMFunctionReturnNode()
  ]}
 New Function_ContainedGraph {nodes [
   n1 = RigVMFunctionEntryNode()
   n2 = &RigVMFunction_Sequence(ExecuteContext=@n1.ExecuteContext)
   n3 = &RigUnit_SphericalPoseReader(
    ActiveRegionScaleFactors=(PositiveWidth=3,NegativeWidth=1,PositiveHeight=1.5,NegativeHeight=1.5)
    ActiveRegionSize=0.12
    Debug=(bDrawDebug=false,bDraw2D=False,bDrawLocalAxes=False,DebugScale=25,DebugSegments=20,DebugThickness=0.25)
    DriverAxis=(X=-1,Y=0,Z=0)
    DriverItem=(Type=Bone,Name="thigh_l")
    FalloffRegionScaleFactors=(PositiveWidth=1,NegativeWidth=1,PositiveHeight=0.75,NegativeHeight=0.75)
    FalloffSize=0.6
    FlipHeightScaling=False
    FlipWidthScaling=False
    OptionalParentItem=(Type=Bone,Name="pelvis")
    RotationOffset=(X=90,Y=0,Z=-15)
    ExecutePin=@n2.A
   )
   > &RigUnit_ModifyTransforms(
    ItemToModify=((Item=(Type=Bone,Name="dyn_bag_frnt_l_base"),Transform=(Rotation=(X=0.1014,Y=-0.1548,Z=0.016,W=0.9826),Translation=(X=0,Y=0,Z=0),Scale3D=(X=1,Y=1,Z=1))))
    WeightMinimum=0
    Weight=@n3.OutputParam
   )
   > n4 = &RigUnit_SphericalPoseReader(
    ActiveRegionScaleFactors=(PositiveWidth=3,NegativeWidth=1.5,PositiveHeight=1.3,NegativeHeight=1.3)
    ActiveRegionSize=0.12
    Debug=(bDrawDebug=false,bDraw2D=False,bDrawLocalAxes=False,DebugScale=25,DebugSegments=20,DebugThickness=0.25)
    DriverAxis=(X=-1,Y=0,Z=0)
    DriverItem=(Type=Bone,Name="thigh_l")
    FalloffRegionScaleFactors=(PositiveWidth=1,NegativeWidth=1,PositiveHeight=0.75,NegativeHeight=0.75)
    FalloffSize=0.5
    FlipHeightScaling=False
    FlipWidthScaling=False
    OptionalParentItem=(Type=Bone,Name="pelvis")
    RotationOffset=(X=130,Y=5,Z=-30)
   )
   > &RigUnit_ModifyTransforms(
    ItemToModify="((Item=(Type=Bone,Name=\"dyn_bag_frnt_l_base\"),Transform=(Rotation=(X=0.1628,Y=-0.3061,Z=0.0911,W=0.9335),Translation=(X=1.7448,Y=-1.5931,Z=0.9002),Scale3D=(X=1,Y=1,Z=1))),(Item=(Type=Bone,Name=\"dyn_bag_side_l_base\"),Transform=(Rotation=(X=0,Y=0,Z=0,W=-1),T ...(truncated)"
    WeightMinimum=0
    Weight=@n4.OutputParam
   )
   > n5 = &RigUnit_SphericalPoseReader(
    ActiveRegionScaleFactors=(PositiveWidth=2,NegativeWidth=1,PositiveHeight=1,NegativeHeight=1)
    ActiveRegionSize=0.15
    Debug=(bDrawDebug=false,bDraw2D=False,bDrawLocalAxes=False,DebugScale=25,DebugSegments=20,DebugThickness=0.25)
    DriverAxis=(X=-1,Y=0,Z=0)
    DriverItem=(Type=Bone,Name="thigh_l")
    FalloffRegionScaleFactors=(PositiveWidth=1,NegativeWidth=1,PositiveHeight=1,NegativeHeight=1)
    FalloffSize=0.4
    FlipHeightScaling=False
    FlipWidthScaling=False
    OptionalParentItem=(Type=Bone,Name="pelvis")
    RotationOffset=(X=180,Y=0,Z=0)
   )
   > &RigUnit_ModifyTransforms(
    ItemToModify="((Item=(Type=Bone,Name=\"dyn_bag_frnt_l_base\"),Transform=(Rotation=(X=-0.1063,Y=-0.487,Z=0.0602,W=0.8648),Translation=(X=0.7134,Y=0.4062,Z=1.81),Scale3D=(X=1,Y=1,Z=1))),(Item=(Type=Bone,Name=\"dyn_bag_side_l_base\"),Transform=(Rotation=(X=-0.1545,Y=0.4408,Z=- ...(truncated)"
    WeightMinimum=0
    Weight=@n5.OutputParam
   )
   > n6 = &RigUnit_SphericalPoseReader(
    ActiveRegionScaleFactors=(PositiveWidth=1,NegativeWidth=1,PositiveHeight=3.5,NegativeHeight=3.5)
    ActiveRegionSize=0.15
    Debug=(bDrawDebug=false,bDraw2D=False,bDrawLocalAxes=False,DebugScale=25,DebugSegments=20,DebugThickness=0.25)
    DriverAxis=(X=1,Y=0,Z=0)
    DriverItem=(Type=Bone,Name="hand_l")
    FalloffRegionScaleFactors=(PositiveWidth=1,NegativeWidth=1,PositiveHeight=1,NegativeHeight=1)
    FalloffSize=0.6
    FlipHeightScaling=False
    FlipWidthScaling=False
    OptionalParentItem=(Type=Bone,Name="None")
    RotationOffset=(X=90,Y=0,Z=-20)
   )
   > &RigUnit_ModifyTransforms(
    ItemToModify="((Item=(Type=Bone,Name=\"dyn_hand_bow_l_01_end_base\"),Transform=(Rotation=(X=-0.0039,Y=-0.1435,Z=-0.4319,W=0.8904),Translation=(X=0,Y=0,Z=0),Scale3D=(X=1,Y=1,Z=1))),(Item=(Type=Bone,Name=\"dyn_hand_bow_l_02_end_base\"),Transform=(Rotation=(X=-0.0529,Y=-0.1647 ...(truncated)"
    WeightMinimum=0
    Weight=@n6.OutputParam
   )
   > RigVMFunctionReturnNode()
   n7 = &RigUnit_SphericalPoseReader(
    ActiveRegionScaleFactors=(PositiveWidth=3,NegativeWidth=1,PositiveHeight=1.5,NegativeHeight=1.5)
    ActiveRegionSize=0.1
    Debug=(bDrawDebug=true,bDraw2D=False,bDrawLocalAxes=False,DebugScale=25,DebugSegments=20,DebugThickness=0.25)
    DriverAxis=(X=1,Y=0,Z=0)
    DriverItem=(Type=Bone,Name="thigh_r")
    FalloffRegionScaleFactors=(PositiveWidth=1,NegativeWidth=1,PositiveHeight=0.75,NegativeHeight=0.75)
    FalloffSize=0.4
    FlipHeightScaling=true
    FlipWidthScaling=true
    OptionalParentItem=(Type=Bone,Name="None")
    RotationOffset=(X=-88.1382,Y=-0.0285,Z=-10)
    ExecutePin=@n2.B
   )
   RigVMFunction_AlphaInterp = &RigVMFunction_AlphaInterp(InterpSpeedDecreasing=7, InterpSpeedIncreasing=7, Scale=1, bClampResult=False, bInterpResult=True, Value=@n7.OutputParam)
   > &RigUnit_ModifyTransforms(
    ItemToModify=((Item=(Type=Bone,Name="dyn_bag_frnt_r_base"),Transform=(Rotation=(X=0.0098,Y=-0.3931,Z=-0.0977,W=0.9142),Translation=(X=1.5322,Y=-0.0005,Z=-0.1041),Scale3D=(X=1,Y=1,Z=1))))
    WeightMinimum=0
    Weight=@RigVMFunction_AlphaInterp.Result
   )
   > n8 = &RigUnit_SphericalPoseReader(
    ActiveRegionScaleFactors=(PositiveWidth=3,NegativeWidth=1.5,PositiveHeight=1.3,NegativeHeight=1.3)
    ActiveRegionSize=0.12
    Debug=(bDrawDebug=false,bDraw2D=False,bDrawLocalAxes=False,DebugScale=25,DebugSegments=20,DebugThickness=0.25)
    DriverAxis=(X=1,Y=0,Z=0)
    DriverItem=(Type=Bone,Name="thigh_r")
    FalloffRegionScaleFactors=(PositiveWidth=1,NegativeWidth=1,PositiveHeight=0.75,NegativeHeight=0.75)
    FalloffSize=0.5
    FlipHeightScaling=true
    FlipWidthScaling=true
    OptionalParentItem=(Type=Bone,Name="None")
    RotationOffset=(X=-50,Y=5,Z=-30)
   )
   > &RigUnit_ModifyTransforms(
    ItemToModify="((Item=(Type=Bone,Name=\"dyn_bag_frnt_r_base\"),Transform=(Rotation=(X=-0.0006,Y=0.1828,Z=0.0729,W=-0.9804),Translation=(X=0.163,Y=0.0433,Z=2),Scale3D=(X=1,Y=1,Z=1))),(Item=(Type=Bone,Name=\"dyn_bag_side_r_base\"),Transform=(Rotation=(X=0.0665,Y=0.271,Z=0.0228 ...(truncated)"
    WeightMinimum=0
    Weight=@n8.OutputParam
   )
   > n9 = &RigUnit_SphericalPoseReader(
    ActiveRegionScaleFactors=(PositiveWidth=2,NegativeWidth=1,PositiveHeight=1,NegativeHeight=1)
    ActiveRegionSize=0.15
    Debug=(bDrawDebug=false,bDraw2D=False,bDrawLocalAxes=False,DebugScale=25,DebugSegments=20,DebugThickness=0.25)
    DriverAxis=(X=1,Y=0,Z=0)
    DriverItem=(Type=Bone,Name="thigh_r")
    FalloffRegionScaleFactors=(PositiveWidth=1,NegativeWidth=1,PositiveHeight=1,NegativeHeight=1)
    FalloffSize=0.4
    FlipHeightScaling=true
    FlipWidthScaling=true
    OptionalParentItem=(Type=Bone,Name="None")
    RotationOffset=(X=0,Y=0,Z=0)
   )
   > &RigUnit_ModifyTransforms(
    ItemToModify="((Item=(Type=Bone,Name=\"dyn_bag_frnt_r_base\"),Transform=(Rotation=(X=0.1921,Y=-0.2925,Z=0.0057,W=0.9368),Translation=(X=1.4182,Y=-0.1306,Z=0),Scale3D=(X=1,Y=1,Z=1))),(Item=(Type=Bone,Name=\"dyn_bag_side_r_base\"),Transform=(Rotation=(X=0.0723,Y=0.392,Z=0.191 ...(truncated)"
    WeightMinimum=0
    Weight=@n9.OutputParam
   )
   > n10 = &RigUnit_SphericalPoseReader(
    ActiveRegionScaleFactors=(PositiveWidth=1,NegativeWidth=1,PositiveHeight=3.5,NegativeHeight=3.5)
    ActiveRegionSize=0.15
    Debug=(bDrawDebug=false,bDraw2D=False,bDrawLocalAxes=False,DebugScale=25,DebugSegments=20,DebugThickness=0.25)
    DriverAxis=(X=-1,Y=0,Z=0)
    DriverItem=(Type=Bone,Name="hand_r")
    FalloffRegionScaleFactors=(PositiveWidth=1,NegativeWidth=1,PositiveHeight=1,NegativeHeight=1)
    FalloffSize=0.6
    FlipHeightScaling=False
    FlipWidthScaling=False
    OptionalParentItem=(Type=Bone,Name="None")
    RotationOffset=(X=-90,Y=0,Z=-20)
   )
   > &RigUnit_ModifyTransforms(
    ItemToModify="((Item=(Type=Bone,Name=\"dyn_hand_bow_r_01_end_base\"),Transform=(Rotation=(X=-0.1198,Y=-0.5358,Z=-0.1823,W=0.8156),Translation=(X=0,Y=0,Z=0),Scale3D=(X=1,Y=1,Z=1))),(Item=(Type=Bone,Name=\"dyn_hand_bow_r_02_end_base\"),Transform=(Rotation=(X=0.1152,Y=0.3401,Z ...(truncated)"
    WeightMinimum=0
    Weight=@n10.OutputParam
   )
  ]}
 RigVMFunctionLibrary {nodes [
   Deform_Body = RigVMCollapseNode(ExecuteContext=())
   Deform_Boots = RigVMCollapseNode(ExecuteContext=())
   Deform_Constraints = RigVMCollapseNode(ExecuteContext=())
  ]}
 RigVMFunction_Sequence_1_ContainedGraph {nodes [
   n1 = RigVMFunctionEntryNode()
   RigVMFunction_Sequence = &RigVMFunction_Sequence(ExecuteContext=@n1.ExecuteContext)
   n2 = &RigVMFunction_Sequence(ExecuteContext=@RigVMFunction_Sequence.B)
   RigVMFunctionReturnNode(A=@RigVMFunction_Sequence.A, B=@n2.A, C=@n2.B)
  ]}
 RigVMModel {nodes [
   RigUnit_BeginExecution = &RigUnit_BeginExecution()
   RigVMFunction_Sequence = RigVMAggregateNode(ExecuteContext=@RigUnit_BeginExecution.ExecutePin)
   FN_DeformRig_FNC = RigVMFunctionReferenceNode(
    DeformRig_ Curve="(deformRig_l_upperarm_up_alpha_72_#1=False,deformRig_r_upperarm_up_alpha_73_#2=False,deformRig_l_upperarm_fr_alpha_74_#3=False,deformRig_r_upperarm_fr_alpha_75_CB51B8 ...(truncated)"
    DeformRig_Bool="(upperarm_1_#4=True,elbow_3_#5=True,wrist_7_#6=True,groin_8_#7=True,glute_17_#8=True,knee_9_40C408324CA261 ...(truncated)"
    DeformRig_Debug=(upperarm_debug_18_#4=False,wrist_debug_20_#6=False,lat_debug_26_#9=False,groinGlute_debug_27_#7=False)
    DeformRig_Value="(upperarm_up_18_#10=(Rotation=(X=0.0,Y=-0.0,Z=0.0,W=1.0),Translation=(X=0.0,Y=0.0,Z=8.0),Scale3D=(X=0.70,Y=1.0,Z=1.0)),upperarm_fwd_fr_19_#11=(Rotation=(X=0.0,Y=-0.0,Z=0.0,W=1.0),Translation=(X=0.0, ...(truncated)"
    ExecuteContext=@RigVMFunction_Sequence.A
   )
   > Deform_Boots = RigVMFunctionReferenceNode()
   > Deform_Constraints = RigVMFunctionReferenceNode()
   > Deform_Body = RigVMFunctionReferenceNode()
  ]}
]
hierarchy [
 root {loc=<0 0 0>, rot=<0 0 0 1>}
 pelvis {parent=root, loc=<0 -0.5301 98.0571>, rot=<0.0127 0.707 -0.0127 -0.707>}
 spine_01 {parent=pelvis, loc=<6.1208 -0.9328 0>, rot=<0 0 0.034 -0.9994>}
 spine_02 {parent=spine_01, loc=<6.5997 -0.2095 0>, rot=<0 0 0.0238 -0.9997>}
 spine_03 {parent=spine_02, loc=<6.4914 -0.1237 0>, rot=<0 0 0.0149 -0.9999>}
 spine_04 {parent=spine_03, loc=<6.9598 0.0794 0>, rot=<0 0 -0.0132 -0.9999>}
 spine_05 {parent=spine_04, loc=<6.9987 0 0>, rot=<0 0 -0.0015 -1>}
 neck_01 {parent=spine_05, loc=<11.3138 0.1529 0>, rot=<0 0 -0.2003 -0.9797>}
 neck_02 {parent=neck_01, loc=<4.5627 0 0>, rot=<0 0 0.0504 -0.9987>}
 head {parent=neck_02, loc=<5.3914 0 0>, rot=<0 0 0.115 -0.9934>}
 clavicle_l {parent=spine_05, loc=<5.4831 -0.1053 -3.2946>, rot=<0.0465 -0.7289 -0.009 -0.6829>}
 upperarm_l {parent=clavicle_l, loc=<9.8658 -1.2136 -0.4105>, rot=<0.0164 -0.3245 -0.0269 -0.9454>}
 lowerarm_l {parent=upperarm_l, loc=<26.465 -0.8345 0.259>, rot=<0.0036 0.0161 -0.2298 -0.9731>}
 hand_l {parent=lowerarm_l, loc=<22.7355 -0.3203 -0.6105>, rot=<-0.6784 -0.0223 0.0735 -0.7307>}
 thumb_01_l {parent=hand_l, loc=<1.3118 -0.3191 -2.459>, rot=<0.6546 -0.4042 0.0664 -0.6354>}
 thumb_02_l {parent=thumb_01_l, loc=<5.9336 -0.0448 0.0401>, rot=<-0.0037 0.0213 0.298 -0.9543>}
 thumb_03_l {parent=thumb_02_l, loc=<2.2223 0.3682 -0.0907>, rot=<0.0118 -0.013 0.0066 -0.9998>}
 index_metacarpal_l {parent=hand_l, loc=<1.6535 0.0183 -2.96>, rot=<0.0158 -0.0124 -0.0173 -0.9996>}
 index_01_l {parent=index_metacarpal_l, loc=<8.3008 0.0357 -0.0524>, rot=<-0.0044 0.0485 0.0924 -0.9945>}
 index_02_l {parent=index_01_l, loc=<3.3587 -0.3004 0.0452>, rot=<-0.002 0.0027 0.3225 -0.9466>}
 index_03_l {parent=index_02_l, loc=<2.5578 -0.123 -0.0252>, rot=<-0.002 0.0067 0.2558 -0.9667>}
 middle_metacarpal_l {parent=hand_l, loc=<2.1856 -0.1725 -0.5984>, rot=<0.006 0.0168 -0.0329 -0.9993>}
 middle_01_l {parent=middle_metacarpal_l, loc=<7.3275 -0.2447 0.0129>, rot=<-0.0208 0.0042 0.0858 -0.9961>}
 middle_02_l {parent=middle_01_l, loc=<4.0123 0.1401 0.1175>, rot=<-0.0218 0.0258 0.3512 -0.9357>}
 middle_03_l {parent=middle_02_l, loc=<2.6077 0.0534 0.1373>, rot=<0.0082 0.0188 0.1708 -0.9851>}
 ring_metacarpal_l {parent=hand_l, loc=<2.2074 -0.3625 1.3638>, rot=<-0.0284 0.0722 -0.0458 -0.9959>}
 ring_01_l {parent=ring_metacarpal_l, loc=<7.0295 -0.6197 -0.1446>, rot=<-0.0489 -0.0252 0.1892 -0.9804>}
 ring_02_l {parent=ring_01_l, loc=<3.6174 0.4647 -0.1027>, rot=<0 0 0.2806 -0.9598>}
 ring_03_l {parent=ring_02_l, loc=<2.4152 0.114 0.0199>, rot=<0 0 0.1189 -0.9929>}
 pinky_metacarpal_l {parent=hand_l, loc=<1.7616 -0.9753 3.3814>, rot=<-0.0397 0.0814 -0.069 -0.9935>}
 pinky_01_l {parent=pinky_metacarpal_l, loc=<7.1709 -0.4831 -0.1094>, rot=<-0.0654 -0.0035 0.2242 -0.9723>}
 pinky_02_l {parent=pinky_01_l, loc=<2.8274 0.0893 -0.1043>, rot=<0.0037 0.0011 0.2854 -0.9584>}
 pinky_03_l {parent=pinky_02_l, loc=<2.2046 0.0307 -0.0078>, rot=<-0.0002 0.0098 0.1445 -0.9895>}
 deform_wrist_dn_base_l {parent=hand_l, loc=<0 -1.5326 0>, rot=<0 0 0 -1>}
 deform_wrist_dn_l {parent=deform_wrist_dn_base_l, loc=<0 0 0>, rot=<0 0 0 1>}
 deform_wrist_up_base_l {parent=hand_l, loc=<0 2.3009 0>, rot=<0 0 0 -1>}
 deform_wrist_up_l {parent=deform_wrist_up_base_l, loc=<0 0 0>, rot=<0 0 0 1>}
 lowerarm_twist_02_l {parent=lowerarm_l, loc=<10.6707 0 0>, rot=<0 0 0 1>}
 lowerarm_twist_01_l {parent=lowerarm_l, loc=<18.316 0 0>, rot=<0 0 0 1>}
 dyn_hand_bow_l_base {parent=lowerarm_twist_01_l, loc=<3.3823 0.2695 2.3477>, rot=<-0.6354 -0.0468 0.0509 -0.7691>}
 dyn_hand_bow_l_02_base {parent=dyn_hand_bow_l_base, loc=<-0.8424 0 -1.248>, rot=<-0.3435 -0.6145 0.6247 -0.3379>}
 dyn_hand_bow_l_02 {parent=dyn_hand_bow_l_02_base, loc=<0 0 0>, rot=<0 0 0 1>}
 dyn_hand_bow_l_01_base {parent=dyn_hand_bow_l_base, loc=<-0.6707 0 1.3276>, rot=<0.3583 0.6039 0.6209 -0.3485>}
 dyn_hand_bow_l_01 {parent=dyn_hand_bow_l_01_base, loc=<0 0 0>, rot=<0 0 0 1>}
 dyn_hand_bow_l_01_end_base {parent=dyn_hand_bow_l_base, loc=<0.0487 0.0135 0.8022>, rot=<0.5581 0.1336 0.24 -0.783>}
 dyn_hand_bow_l_01_end {parent=dyn_hand_bow_l_01_end_base, loc=<0 0 0>, rot=<0 0 0 1>}
 dyn_hand_bow_l_02_end_base {parent=dyn_hand_bow_l_base, loc=<-0.1194 -0.096 -0.5182>, rot=<-0.5367 -0.1519 0.2636 -0.787>}
 dyn_hand_bow_l_02_end {parent=dyn_hand_bow_l_02_end_base, loc=<0 0 0>, rot=<0 0 0 1>}
 deform_elbow_in_base_l {parent=lowerarm_l, loc=<0.1834 3.6062 -0.2572>, rot=<0 0 0 -1>}
 deform_elbow_in_l {parent=deform_elbow_in_base_l, loc=<0 0 0>, rot=<0 0 0 1>}
 deform_elbow_out_base_l {parent=lowerarm_l, loc=<0 -4 0>, rot=<0 0 0 1>}
 deform_elbow_out_l {parent=deform_elbow_out_base_l, loc=<0 0 0>, rot=<0 0 0 1>}
 upperarm_twist_01_l {parent=upperarm_l, loc=<7.2135 -0.1134 0>, rot=<0 0 0 1>}
 deform_upperarm_bk_base_l {parent=upperarm_twist_01_l, loc=<-6.4053 -4.3746 0.314>, rot=<0 0 0 -1>}
 deform_upperarm_bk_l {parent=deform_upperarm_bk_base_l, loc=<0 0 0>, rot=<0 0 0 1>}
 deform_upperarm_fr_base_l {parent=upperarm_twist_01_l, loc=<-5.6312 5.092 0.3354>, rot=<0 0 0 -1>}
 deform_upperarm_fr_l {parent=deform_upperarm_fr_base_l, loc=<0 0 0>, rot=<0 0 0 1>}
 deform_upperarm_up_base_l {parent=upperarm_twist_01_l, loc=<-8.891 0.4825 3.9517>, rot=<0 0 0 -1>}
 deform_upperarm_up_l {parent=deform_upperarm_up_base_l, loc=<0 0 0>, rot=<0 0 0 1>}
 upperarm_twist_02_l {parent=upperarm_l, loc=<19.7998 -0.1134 0>, rot=<0 0 0 1>}
 clavicle_r {parent=spine_05, loc=<5.4831 -0.1053 3.2946>, rot=<0.7289 0.0465 -0.6829 0.009>}
 upperarm_r {parent=clavicle_r, loc=<-9.8658 1.2136 0.4105>, rot=<0.0164 -0.3245 -0.0269 -0.9454>}
 lowerarm_r {parent=upperarm_r, loc=<-26.465 0.8345 -0.259>, rot=<0.0036 0.0161 -0.2298 -0.9731>}
 hand_r {parent=lowerarm_r, loc=<-22.7355 0.3203 0.6106>, rot=<-0.6784 -0.0223 0.0735 -0.7307>}
 thumb_01_r {parent=hand_r, loc=<-1.3118 0.3191 2.459>, rot=<0.6546 -0.4042 0.0664 -0.6354>}
 thumb_02_r {parent=thumb_01_r, loc=<-5.9336 0.0448 -0.0401>, rot=<-0.0037 0.0213 0.298 -0.9543>}
 thumb_03_r {parent=thumb_02_r, loc=<-2.2223 -0.3682 0.0907>, rot=<0.0118 -0.013 0.0066 -0.9998>}
 index_metacarpal_r {parent=hand_r, loc=<-1.6535 -0.0183 2.96>, rot=<0.0158 -0.0124 -0.0173 -0.9996>}
 index_01_r {parent=index_metacarpal_r, loc=<-8.3008 -0.0357 0.0524>, rot=<-0.0044 0.0485 0.0924 -0.9945>}
 index_02_r {parent=index_01_r, loc=<-3.3587 0.3004 -0.0452>, rot=<-0.002 0.0027 0.3225 -0.9466>}
 index_03_r {parent=index_02_r, loc=<-2.5578 0.123 0.0252>, rot=<-0.002 0.0067 0.2558 -0.9667>}
 middle_metacarpal_r {parent=hand_r, loc=<-2.1856 0.1725 0.5984>, rot=<0.006 0.0168 -0.0329 -0.9993>}
 middle_01_r {parent=middle_metacarpal_r, loc=<-7.3275 0.2447 -0.0129>, rot=<-0.0208 0.0042 0.0858 -0.9961>}
 middle_02_r {parent=middle_01_r, loc=<-4.0123 -0.1401 -0.1175>, rot=<-0.0218 0.0258 0.3512 -0.9357>}
 middle_03_r {parent=middle_02_r, loc=<-2.6077 -0.0534 -0.1373>, rot=<0.0082 0.0188 0.1708 -0.9851>}
 ring_metacarpal_r {parent=hand_r, loc=<-2.2074 0.3625 -1.3638>, rot=<-0.0284 0.0722 -0.0458 -0.9959>}
 ring_01_r {parent=ring_metacarpal_r, loc=<-7.0295 0.6197 0.1446>, rot=<-0.0489 -0.0252 0.1892 -0.9804>}
 ring_02_r {parent=ring_01_r, loc=<-3.6174 -0.4647 0.1027>, rot=<0 0 0.2806 -0.9598>}
 ring_03_r {parent=ring_02_r, loc=<-2.4152 -0.114 -0.0199>, rot=<0 0 0.1189 -0.9929>}
 pinky_metacarpal_r {parent=hand_r, loc=<-1.7616 0.9753 -3.3814>, rot=<-0.0397 0.0814 -0.069 -0.9935>}
 pinky_01_r {parent=pinky_metacarpal_r, loc=<-7.1709 0.4831 0.1094>, rot=<-0.0654 -0.0035 0.2242 -0.9723>}
 pinky_02_r {parent=pinky_01_r, loc=<-2.8274 -0.0893 0.1043>, rot=<0.0037 0.0011 0.2854 -0.9584>}
 pinky_03_r {parent=pinky_02_r, loc=<-2.2046 -0.0307 0.0078>, rot=<-0.0002 0.0098 0.1445 -0.9895>}
 deform_wrist_dn_base_r {parent=hand_r, loc=<-0.0001 1.5328 0>, rot=<0 0 0 -1>}
 deform_wrist_dn_r {parent=deform_wrist_dn_base_r, loc=<0 0 0>, rot=<0 0 0 1>}
 deform_wrist_up_base_r {parent=hand_r, loc=<-0.0002 -2.3007 -0.0001>, rot=<0 0 0 -1>}
 deform_wrist_up_r {parent=deform_wrist_up_base_r, loc=<0 0 0>, rot=<0 0 0 1>}
 lowerarm_twist_02_r {parent=lowerarm_r, loc=<-10.6707 0 0>, rot=<0 0 0 -1>}
 lowerarm_twist_01_r {parent=lowerarm_r, loc=<-18.316 0 0>, rot=<0 0 0 -1>}
 dyn_hand_bow_r_base {parent=lowerarm_twist_01_r, loc=<-3.3821 -0.2695 -2.3478>, rot=<-0.6354 -0.0468 0.0509 -0.7691>}
 dyn_hand_bow_r_02_base {parent=dyn_hand_bow_r_base, loc=<0.8419 0.0004 1.2479>, rot=<-0.3435 -0.6145 0.6247 -0.3379>}
 dyn_hand_bow_r_02 {parent=dyn_hand_bow_r_02_base, loc=<0 0 0>, rot=<0 0 0 -1>}
 dyn_hand_bow_r_01_base {parent=dyn_hand_bow_r_base, loc=<0.6702 0.0004 -1.3278>, rot=<0.3583 0.6039 0.6209 -0.3485>}
 dyn_hand_bow_r_01 {parent=dyn_hand_bow_r_01_base, loc=<0 0 0>, rot=<0 0 0 1>}
 dyn_hand_bow_r_01_end_base {parent=dyn_hand_bow_r_base, loc=<-0.0493 -0.0131 -0.8024>, rot=<0.5581 0.1336 0.24 -0.783>}
 dyn_hand_bow_r_01_end {parent=dyn_hand_bow_r_01_end_base, loc=<0 0 0>, rot=<0 0 0 1>}
 dyn_hand_bow_r_02_end_base {parent=dyn_hand_bow_r_base, loc=<0.1191 0.0962 0.5182>, rot=<-0.5367 -0.1519 0.2636 -0.787>}
 dyn_hand_bow_r_02_end {parent=dyn_hand_bow_r_02_end_base, loc=<0 0 0>, rot=<0 0 0 1>}
 deform_elbow_in_base_r {parent=lowerarm_r, loc=<-0.1834 -3.6062 0.2572>, rot=<0 0 0 -1>}
 deform_elbow_in_r {parent=deform_elbow_in_base_r, loc=<0 0 0>, rot=<0 0 0 1>}
 deform_elbow_out_base_r {parent=lowerarm_r, loc=<0 4 -0.0001>, rot=<0 0 0 -1>}
 deform_elbow_out_r {parent=deform_elbow_out_base_r, loc=<0 0 0>, rot=<0 0 0 1>}
 upperarm_twist_01_r {parent=upperarm_r, loc=<-7.2135 0.1134 0>, rot=<0 0 0 -1>}
 deform_upperarm_bk_base_r {parent=upperarm_twist_01_r, loc=<6.4053 4.3746 -0.314>, rot=<0 0 0 -1>}
 deform_upperarm_bk_r {parent=deform_upperarm_bk_base_r, loc=<0 0 0>, rot=<0 0 0 1>}
 deform_upperarm_fr_base_r {parent=upperarm_twist_01_r, loc=<5.6312 -5.092 -0.3354>, rot=<0 0 0 -1>}
 deform_upperarm_fr_r {parent=deform_upperarm_fr_base_r, loc=<0 0 0>, rot=<0 0 0 1>}
 deform_upperarm_up_base_r {parent=upperarm_twist_01_r, loc=<8.891 -0.4825 -3.9517>, rot=<0 0 0 -1>}
 deform_upperarm_up_r {parent=deform_upperarm_up_base_r, loc=<0 0 0>, rot=<0 0 0 1>}
 upperarm_twist_02_r {parent=upperarm_r, loc=<-19.7998 0.1134 0>, rot=<0 0 0 -1>}
 deform_button {parent=clavicle_r, loc=<-6.5405 -4.4861 -2.5535>, rot=<-0.551 0.3324 0.7257 -0.2434>}
 attach_backpack {parent=spine_05, loc=<-0.1228 -13.4724 0>, rot=<-0.0093 -0.707 0.0093 -0.707>}
 Backpack_BR {parent=attach_backpack, type=Null, loc=<0 5.0316 0>, rot=<0 0 0 1>}
 attach_cape {parent=spine_05, loc=<9.0172 -11.7137 -0.0001>, rot=<0.0616 -0.7044 -0.0616 -0.7044>}
 cape {parent=attach_cape, type=Null, loc=<0 2.5692 0>, rot=<0 0 0 1>}
 deform_pec_aim_base_l {parent=spine_04, loc=<3.2456 9.3309 -1.2237>, rot=<0.0603 -0.4033 0.3318 -0.8507>}
 deform_pec_aim_l {parent=deform_pec_aim_base_l, loc=<0 0 0>, rot=<0 0 0 1>}
 deform_pec_base_l {parent=deform_pec_aim_l, loc=<9.3665 0 0>, rot=<0 0 0 1>}
 deform_pec_l {parent=deform_pec_base_l, loc=<0 0 0>, rot=<0 0 0 1>}
 deform_pec_aim_base_r {parent=spine_04, loc=<3.2452 9.3309 1.2237>, rot=<-0.4033 -0.0603 0.8507 0.3318>}
 deform_pec_aim_r {parent=deform_pec_aim_base_r, loc=<0 0 0>, rot=<0 0 0 1>}
 deform_pec_base_r {parent=deform_pec_aim_r, loc=<-9.3667 -0.0003 -0.0004>, rot=<0 0 0 1>}
 deform_pec_r {parent=deform_pec_base_r, loc=<0 0 0>, rot=<0 0 0 1>}
 dyn_torso_l_base {parent=spine_04, loc=<0.5408 4.9573 -5.2979>, rot=<-0.4316 -0.5601 -0.4767 -0.5223>}
 dyn_torso_l {parent=dyn_torso_l_base, loc=<0 0 0>, rot=<0 0 0 1>}
 dyn_torso_r_base {parent=spine_04, loc=<0.5408 4.9573 5.298>, rot=<0.501 0.499 0.5408 0.4556>}
 dyn_torso_r {parent=dyn_torso_r_base, loc=<0 0 0>, rot=<0 0 0 1>}
 thigh_l {parent=pelvis, loc=<-1.0855 -0.2109 -8.7053>, rot=<0.0009 0.0355 0.0155 -0.9993>}
 calf_l {parent=thigh_l, loc=<-42.5284 -0.0019 -0.7297>, rot=<0.0006 -0.0161 -0.0391 -0.9991>}
 foot_l {parent=calf_l, loc=<-46.7401 -2.5014 -0.592>, rot=<0.0427 -0.0238 0.0353 -0.9982>}
 ball_l {parent=foot_l, loc=<-3.5658 10.8066 -0.1544>, rot=<0.0001 -0.0001 -0.7063 -0.7079>}
 calf_twist_02_l {parent=calf_l, loc=<-19.1546 -1.7943 -0.2638>, rot=<0.0237 0.0095 -0.0035 -0.9997>}
 calf_twist_01_l {parent=calf_l, loc=<-35.8483 -1.9197 -0.5794>, rot=<0.0237 0.0095 -0.0035 -0.9997>}
 foot_support_l {parent=calf_twist_01_l, loc=<-8.1518 -4.8502 0.133>, rot=<0.0193 -0.0343 0.0378 -0.9985>}
 dyn_foot_bow_l_base {parent=calf_twist_01_l, loc=<-2.613 7.0822 -0.1077>, rot=<0.0193 -0.0343 0.0378 -0.9985>}
 dyn_foot_bow_l_01_base {parent=dyn_foot_bow_l_base, loc=<-0.372 0.4545 -2.6757>, rot=<0.0014 -0.6978 0.0088 -0.7163>}
 dyn_foot_bow_l_01 {parent=dyn_foot_bow_l_01_base, loc=<0 0 0>, rot=<0 0 0 1>}
 dyn_foot_bow_l_02_base {parent=dyn_foot_bow_l_base, loc=<-0.2308 0.4558 2.6461>, rot=<0.7054 -0.0007 0.7088 0.0068>}
 dyn_foot_bow_l_02 {parent=dyn_foot_bow_l_02_base, loc=<0 0 0>, rot=<0 0 0 1>}
 dyn_foot_bow_l_01_end_base {parent=dyn_foot_bow_l_base, loc=<-1.5878 0.559 -1.3764>, rot=<0.0004 -0.9068 0.0136 -0.4213>}
 dyn_foot_bow_l_01_end {parent=dyn_foot_bow_l_01_end_base, loc=<0 0 0>, rot=<0 0 0 1>}
 dyn_foot_bow_l_02_end_base {parent=dyn_foot_bow_l_base, loc=<-1.4479 0.1702 1.2265>, rot=<0.4489 -0.2739 0.8399 0.1344>}
 dyn_foot_bow_l_02_end {parent=dyn_foot_bow_l_02_end_base, loc=<0 0 0>, rot=<0 0 0 1>}
 deform_knee_in_base_l {parent=calf_l, loc=<0.8031 -5.0231 0.2456>, rot=<0 0 0 -1>}
 deform_knee_in_l {parent=deform_knee_in_base_l, loc=<0 0 0>, rot=<0 0 0 1>}
 deform_knee_out_base_l {parent=calf_l, loc=<0 4 0>, rot=<0 0 0 -1>}
 deform_knee_out_l {parent=deform_knee_out_base_l, loc=<0 0 0>, rot=<0 0 0 1>}
 thigh_twist_01_l {parent=thigh_l, loc=<-27.952 0 0>, rot=<0.0013 0 0 -1>}
 deform_groin_base_l {parent=thigh_l, loc=<0.4801 5.6776 -0.2486>, rot=<0 0 0 -1>}
 deform_groin_l {parent=deform_groin_base_l, loc=<0 0 0>, rot=<0 0 0 1>}
 thigh_r {parent=pelvis, loc=<-1.0855 -0.2109 8.7053>, rot=<0.0355 -0.0009 0.9993 0.0155>}
 calf_r {parent=thigh_r, loc=<42.5284 0.0019 0.7297>, rot=<0.0006 -0.0161 -0.0391 -0.9991>}
 foot_r {parent=calf_r, loc=<46.7401 2.5014 0.592>, rot=<0.0427 -0.0238 0.0353 -0.9982>}
 ball_r {parent=foot_r, loc=<3.5658 -10.8066 0.1544>, rot=<0.0001 -0.0001 -0.7063 -0.7079>}
 foot_support_r {parent=foot_r, loc=<-3.0644 4.1208 0>, rot=<0 0 0 -1>}
 calf_twist_02_r {parent=calf_r, loc=<19.1546 1.7943 0.2638>, rot=<0.0237 0.0095 -0.0035 -0.9997>}
 calf_twist_01_r {parent=calf_r, loc=<35.8483 1.9197 0.5794>, rot=<0.0237 0.0095 -0.0035 -0.9997>}
 dyn_foot_bow_r_base {parent=calf_twist_01_r, loc=<2.613 -7.0822 0.1077>, rot=<0.0193 -0.0343 0.0378 -0.9985>}
 dyn_foot_bow_r_01_base {parent=dyn_foot_bow_r_base, loc=<0.3721 -0.4545 2.6757>, rot=<0.0014 -0.6978 0.0088 -0.7163>}
 dyn_foot_bow_r_01 {parent=dyn_foot_bow_r_01_base, loc=<0 0 0>, rot=<0 0 0 -1>}
 dyn_foot_bow_r_02_base {parent=dyn_foot_bow_r_base, loc=<0.2308 -0.4558 -2.6461>, rot=<0.7054 -0.0007 0.7088 0.0068>}
 dyn_foot_bow_r_02 {parent=dyn_foot_bow_r_02_base, loc=<0 0 0>, rot=<0 0 0 -1>}
 dyn_foot_bow_r_01_end_base {parent=dyn_foot_bow_r_base, loc=<1.5878 -0.559 1.3763>, rot=<0.0004 -0.9068 0.0136 -0.4213>}
 dyn_foot_bow_r_01_end {parent=dyn_foot_bow_r_01_end_base, loc=<0 0 0>, rot=<0 0 0 1>}
 dyn_foot_bow_r_02_end_base {parent=dyn_foot_bow_r_base, loc=<1.4478 -0.1702 -1.2265>, rot=<0.4125 0.4002 0.7875 -0.2224>}
 dyn_foot_bow_r_02_end {parent=dyn_foot_bow_r_02_end_base, loc=<0 0 0>, rot=<0 0 0 1>}
 deform_knee_in_base_r {parent=calf_r, loc=<-0.8031 5.0231 -0.2457>, rot=<0 0 0 -1>}
 deform_knee_in_r {parent=deform_knee_in_base_r, loc=<0 0 0>, rot=<0 0 0 1>}
 deform_knee_out_base_r {parent=calf_r, loc=<0 -4 0>, rot=<0 0 0 -1>}
 deform_knee_out_r {parent=deform_knee_out_base_r, loc=<0 0 0>, rot=<0 0 0 1>}
 thigh_twist_01_r {parent=thigh_r, loc=<27.952 0 0>, rot=<0.0013 0 0 -1>}
 deform_groin_base_r {parent=thigh_r, loc=<-0.4801 -5.6776 0.2486>, rot=<0 0 0 -1>}
 deform_groin_r {parent=deform_groin_base_r, loc=<0 0 0>, rot=<0 0 0 1>}
 deform_glute_base_l {parent=pelvis, loc=<-7.3466 -9.0213 -9.1245>, rot=<0.0172 0.031 -0.4477 -0.8935>}
 deform_glute_l {parent=deform_glute_base_l, loc=<0 0 0>, rot=<0 0 0 1>}
 deform_glute_base_r {parent=pelvis, loc=<-7.3464 -9.0213 9.1245>, rot=<0.031 -0.0172 0.8935 -0.4477>}
 deform_glute_r {parent=deform_glute_base_r, loc=<0 0 0>, rot=<0 0 0 1>}
 dyn_bag_frnt_l_base {parent=pelvis, loc=<0.9418 3.3659 -15.527>, rot=<0.0551 -0.9778 -0.1801 -0.0915>}
 dyn_bag_frnt_l_01 {parent=dyn_bag_frnt_l_base, loc=<0 0 0>, rot=<0 0 0 1>}
 dyn_bag_side_l_base {parent=pelvis, loc=<1.6464 -3.6643 -15.8163>, rot=<0.0285 -0.9878 0.1028 -0.1134>}
 dyn_bag_side_l_01 {parent=dyn_bag_side_l_base, loc=<0 0 0>, rot=<0 0 0 1>}
 dyn_bag_frnt_r_base {parent=pelvis, loc=<-0.2069 6.9966 13.1505>, rot=<0.0866 0.4073 0.9084 -0.0367>}
 dyn_bag_frnt_r_01 {parent=dyn_bag_frnt_r_base, loc=<0 0 0>, rot=<0 0 0 1>}
 dyn_bag_bk_r_base {parent=pelvis, loc=<1.5514 -6.7249 15.1149>, rot=<0.0688 -0.2918 0.9535 -0.0309>}
 dyn_bag_bk_r_01 {parent=dyn_bag_bk_r_base, loc=<0 0 0>, rot=<0 0 0 1>}
 dyn_bag_side_r_base {parent=pelvis, loc=<0.8902 0.7596 16.3824>, rot=<0.0827 -0.0086 0.9964 -0.018>}
 dyn_bag_side_r_01 {parent=dyn_bag_side_r_base, loc=<0 0 0>, rot=<0 0 0 1>}
 deform_buckle_front {parent=pelvis, loc=<-1.1823 9.7192 -0.0012>, rot=<-0.5089 -0.491 -0.491 -0.5089>}
 deform_buckle_back {parent=pelvis, loc=<2.8466 -12.6462 -0.0949>, rot=<0.5819 -0.4013 0.4008 -0.5828>}
]
