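# Physics asset for the F_MED_FondBreak female medium player body cosmetic, covering five dynamic bag bones at the front, sides and back right.
# Each bag has a capsule on its 01 bone, constrained to a tiny capsule on its base bone with narrow swing and very tight twist limits.
# Pelvis and both thighs carry larger capsules with no constraints, forming simple body volumes alongside the bags.
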PhysicsAsset /BRCosmetics/Characters/Player/Female/Medium/Bodies/F_MED_FondBreak/Meshes/F_MED_FondBreak_Physics
bodies {
 dyn_bag_bk_r_01 [{
   shapeType=Capsule
   length=12.5222
   radius=5.4287
   shapeName=dyn_bag_bk_r_01_capsule
   center=<7.6147 -0.0907 2.5672>
   rotation=<rot 0 90.2767 -96.4912>
  }]
 dyn_bag_bk_r_base [{shapeType=Capsule, length=0.505, radius=0.505, shapeName=dyn_bag_bk_r_base_capsule, rotation=<rot 0 -90 0>}]
 dyn_bag_frnt_l_01 [{
   shapeType=Capsule
   length=7.4751
   radius=5.4287
   shapeName=dyn_bag_frnt_l_01_capsule
   center=<3.4159 -0.0068 3.9782>
   rotation=<rot 0 89.4963 -94.6617>
  }]
 dyn_bag_frnt_l_base [{shapeType=Capsule, length=0.505, radius=0.505, shapeName=dyn_bag_frnt_l_base_capsule, rotation=<rot 0 -90 0>}]
 dyn_bag_frnt_r_01 [{
   shapeType=Capsule
   length=12.4231
   radius=5.5278
   shapeName=dyn_bag_frnt_r_01_capsule
   center=<7.7161 -0.041 4.4366>
   rotation=<rot -0.0038 91.0553 -92.9113>
  }]
 dyn_bag_frnt_r_base [{shapeType=Capsule, length=0.505, radius=0.505, shapeName=dyn_bag_frnt_r_base_capsule, rotation=<rot 0 -90 0>}]
 dyn_bag_side_l_01 [{
   shapeType=Capsule
   length=7.4481
   radius=5.4287
   shapeName=dyn_bag_side_l_01_capsule
   center=<3.5623 0.3133 3.7062>
   rotation=<rot 0 -90.5028 97.1238>
  }]
 dyn_bag_side_l_base [{shapeType=Capsule, length=0.505, radius=0.505, shapeName=dyn_bag_side_l_base_capsule, rotation=<rot 0 -90 0>}]
 dyn_bag_side_r_01 [{
   shapeType=Capsule
   length=12.5222
   radius=5.4287
   shapeName=dyn_bag_side_r_01_capsule
   center=<7.246 0.0681 3.9302>
   rotation=<rot 0 91.4966 -95.6484>
  }]
 dyn_bag_side_r_base [{shapeType=Capsule, length=0.505, radius=0.505, shapeName=dyn_bag_side_r_base_capsule, rotation=<rot 0 -90 0>}]
 pelvis [{shapeType=Capsule, length=9.441, radius=10.3016, shapeName=pelvis_capsule, center=<-3.2281 -0.0549 -0.037>}]
 thigh_l [{shapeType=Capsule, length=36.3955, radius=7.2952, shapeName=thigh_l_capsule, center=<-19.3294 -0.241 0.6546>, rotation=<rot -90 0 0>}]
 thigh_r [{shapeType=Capsule, length=36.4092, radius=7.2952, shapeName=thigh_r_capsule, center=<19.217 -0.5675 -0.4256>, rotation=<rot 90 0 0>}]
}
constraints [
 {bone1Name=dyn_bag_bk_r_01, bone2Name=dyn_bag_bk_r_base, swing1LimitDegrees=10, swing2LimitDegrees=20, twistLimitDegrees=2}
 {bone1Name=dyn_bag_frnt_l_01, bone2Name=dyn_bag_frnt_l_base, swing1LimitDegrees=10, swing2LimitDegrees=20, twistLimitDegrees=2}
 {bone1Name=dyn_bag_frnt_r_01, bone2Name=dyn_bag_frnt_r_base, swing1LimitDegrees=10, swing2LimitDegrees=20, twistLimitDegrees=2}
 {bone1Name=dyn_bag_side_l_01, bone2Name=dyn_bag_side_l_base, swing1LimitDegrees=10, swing2LimitDegrees=20, twistLimitDegrees=2}
 {bone1Name=dyn_bag_side_r_01, bone2Name=dyn_bag_side_r_base, swing1LimitDegrees=10, swing2LimitDegrees=20, twistLimitDegrees=2}
]
