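# Physics asset for the face accessory part of the FondBreak female medium player cosmetic, providing capsule physics for its choker bones.
# Two constraints chain dyn_choker_base, dyn_choker_01 and dyn_choker_02, allowing limited, asymmetric swing with a much tighter twist.
# Unconstrained capsules on neck_01, spine_04 and spine_05 give the neck and upper torso collision shapes, with the largest on spine_05.
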PhysicsAsset /BRCosmetics/Characters/Player/Female/Medium/Bodies/F_MED_FondBreak/Meshes/Parts/F_MED_FondBreak_FaceAcc_Physics
bodies {
 dyn_choker_01 [{
   shapeType=Capsule
   length=1.7791
   radius=0.8446
   shapeName=dyn_choker_01_capsule
   center=<1.2344 0.6543 0.0025>
   rotation=<rot 0.0748 -94.0485 -88.764>
  }]
 dyn_choker_02 [{
   shapeType=Capsule
   length=1.8389
   radius=2.2573
   shapeName=dyn_choker_02_capsule
   center=<1.6842 2.0778 -0.0183>
   rotation=<rot 0.0248 89.4354 -88.1141>
  }]
 dyn_choker_base [{shapeType=Capsule, length=0.505, radius=0.505, shapeName=dyn_choker_base_capsule, rotation=<rot 0 -90 0>}]
 neck_01 [{shapeType=Capsule, length=4.861, radius=4.3258, shapeName=neck_01_capsule, center=<0 0.6424 0>, rotation=<rot -90 0 0>}]
 spine_04 [{shapeType=Capsule, length=0.505, radius=0.505, shapeName=spine_04_capsule, rotation=<rot 0 -90 0>}]
 spine_05 [{
   shapeType=Capsule
   length=12.1673
   radius=10.555
   shapeName=spine_05_capsule
   center=<-1.5118 -1.4774 0.0036>
   rotation=<rot -0.0338 -90 0.013>
  }]
}
constraints [
 {bone1Name=dyn_choker_01, bone2Name=dyn_choker_base, swing1LimitDegrees=45, swing2LimitDegrees=25, twistLimitDegrees=6}
 {bone1Name=dyn_choker_02, bone2Name=dyn_choker_01, swing1LimitDegrees=45, swing2LimitDegrees=25, twistLimitDegrees=6}
]
