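# Physics asset for the ribbons mesh part of the female medium FondBreak player cosmetic, defining secondary motion for ribbons and side hair.
# Left and right chains of capsules hang from small base bodies, four segments per ribbon and two per hair lock, linked by constraints.
# Ribbons swing more widely than hair, and both have tight twist limits.
# Unconstrained head, clavicle and right upper arm capsules supply body shapes, with no left upper arm counterpart.
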PhysicsAsset /BRCosmetics/Characters/Player/Female/Medium/Bodies/F_MED_FondBreak/Meshes/Parts/F_MED_FondBreak_Ribbons_Physics
bodies {
 clavicle_l [{
   shapeType=Capsule
   length=13.1029
   radius=7.343
   shapeName=clavicle_l_capsule
   center=<2.4958 0 -1.1586>
   rotation=<rot 81.6571 -0.0007 -0.0007>
  }]
 clavicle_r [{
   shapeType=Capsule
   length=8.8474
   radius=7.343
   shapeName=clavicle_r_capsule
   center=<-1.5093 0.6903 1.0432>
   rotation=<rot 80.2204 -0.0009 -0.001>
  }]
 dyn_hair_side_l_01 [{
   shapeType=Capsule
   length=4.7662
   radius=1.7051
   shapeName=dyn_hair_side_l_01_capsule
   center=<2.1936 1.6738 0.977>
   rotation=<rot 87.6274 0.0003 -2.7926>
  }]
 dyn_hair_side_l_02 [{
   shapeType=Capsule
   length=7.3439
   radius=2.5935
   shapeName=dyn_hair_side_l_02_capsule
   center=<3.4809 2.0756 0.8665>
   rotation=<rot 80.7962 0.0001 0>
  }]
 dyn_hair_side_l_base [{shapeType=Capsule, length=0.505, radius=0.505, shapeName=dyn_hair_side_l_base_capsule, rotation=<rot 0 -90 0>}]
 dyn_hair_side_r_01 [{
   shapeType=Capsule
   length=4.3679
   radius=1.6091
   shapeName=dyn_hair_side_r_01_capsule
   center=<-1.7969 0 -0.8627>
   rotation=<rot 87.358 0.0003 0.0001>
  }]
 dyn_hair_side_r_02 [{
   shapeType=Capsule
   length=7.3168
   radius=2.3427
   shapeName=dyn_hair_side_r_02_capsule
   center=<-2.8778 -1.4396 -0.7926>
   rotation=<rot 80.191 0.0001 -4.4493>
  }]
 dyn_hair_side_r_base [{shapeType=Capsule, length=0.505, radius=0.505, shapeName=dyn_hair_side_r_base_capsule, rotation=<rot 0 -90 0>}]
 dyn_ribbon_l_01 [{shapeType=Capsule, length=6.4494, radius=1.4154, shapeName=dyn_ribbon_l_01_capsule, center=<4.06 0 0>, rotation=<rot -90 0 0>}]
 dyn_ribbon_l_02 [{shapeType=Capsule, length=6.008, radius=2.0176, shapeName=dyn_ribbon_l_02_capsule, center=<4.06 0 0>, rotation=<rot -90 0 0>}]
 dyn_ribbon_l_03 [{shapeType=Capsule, length=6.0877, radius=2.5435, shapeName=dyn_ribbon_l_03_capsule, center=<4.06 0 0>, rotation=<rot -90 0 0>}]
 dyn_ribbon_l_04 [{shapeType=Capsule, length=8.5481, radius=3, shapeName=dyn_ribbon_l_04_capsule, center=<4.06 0 0>, rotation=<rot -90 0 0>}]
 dyn_ribbon_l_base [{shapeType=Capsule, length=0.505, radius=0.505, shapeName=dyn_ribbon_l_base_capsule, rotation=<rot 0 -90 0>}]
 dyn_ribbon_r_01 [{shapeType=Capsule, length=6.7383, radius=1.3626, shapeName=dyn_ribbon_r_01_capsule, center=<-3.9119 0 0>, rotation=<rot 90 0 0>}]
 dyn_ribbon_r_02 [{shapeType=Capsule, length=6.3621, radius=1.7164, shapeName=dyn_ribbon_r_02_capsule, center=<-3.9119 0 0>, rotation=<rot 90 0 0>}]
 dyn_ribbon_r_03 [{shapeType=Capsule, length=7.0069, radius=1.6243, shapeName=dyn_ribbon_r_03_capsule, center=<-3.9119 0 0>, rotation=<rot 90 0 0>}]
 dyn_ribbon_r_04 [{shapeType=Capsule, length=7.432, radius=3, shapeName=dyn_ribbon_r_04_capsule, center=<-4.2918 0 0>, rotation=<rot 90 0 0>}]
 dyn_ribbon_r_base [{shapeType=Capsule, length=0.505, radius=0.505, shapeName=dyn_ribbon_r_base_capsule, rotation=<rot 0 -90 0>}]
 head [{shapeType=Capsule, length=14.2969, radius=7.4085, shapeName=head_capsule, rotation=<rot 5.0944 107.3425 -90>}]
 upperarm_r [{shapeType=Capsule, length=17.4539, radius=4.8879, shapeName=upperarm_r_capsule, center=<-7.2916 0 1.1436>, rotation=<rot 90 0 0>}]
}
constraints [
 {bone1Name=dyn_ribbon_l_01, bone2Name=dyn_ribbon_l_base, swing1LimitDegrees=40, swing2LimitDegrees=40, twistLimitDegrees=4}
 {bone1Name=dyn_ribbon_l_02, bone2Name=dyn_ribbon_l_01, swing1LimitDegrees=40, swing2LimitDegrees=40, twistLimitDegrees=4}
 {bone1Name=dyn_ribbon_l_03, bone2Name=dyn_ribbon_l_02, swing1LimitDegrees=40, swing2LimitDegrees=40, twistLimitDegrees=4}
 {bone1Name=dyn_ribbon_l_04, bone2Name=dyn_ribbon_l_03, swing1LimitDegrees=40, swing2LimitDegrees=40, twistLimitDegrees=4}
 {bone1Name=dyn_ribbon_r_01, bone2Name=dyn_ribbon_r_base, swing1LimitDegrees=40, swing2LimitDegrees=40, twistLimitDegrees=4}
 {bone1Name=dyn_ribbon_r_02, bone2Name=dyn_ribbon_r_01, swing1LimitDegrees=40, swing2LimitDegrees=40, twistLimitDegrees=4}
 {bone1Name=dyn_ribbon_r_03, bone2Name=dyn_ribbon_r_02, swing1LimitDegrees=40, swing2LimitDegrees=40, twistLimitDegrees=4}
 {bone1Name=dyn_ribbon_r_04, bone2Name=dyn_ribbon_r_03, swing1LimitDegrees=40, swing2LimitDegrees=40, twistLimitDegrees=4}
 {bone1Name=dyn_hair_side_l_01, bone2Name=dyn_hair_side_l_base, swing1LimitDegrees=15, swing2LimitDegrees=15, twistLimitDegrees=2}
 {bone1Name=dyn_hair_side_l_02, bone2Name=dyn_hair_side_l_01, swing1LimitDegrees=15, swing2LimitDegrees=15, twistLimitDegrees=2}
 {bone1Name=dyn_hair_side_r_01, bone2Name=dyn_hair_side_r_base, swing1LimitDegrees=15, swing2LimitDegrees=15, twistLimitDegrees=2}
 {bone1Name=dyn_hair_side_r_02, bone2Name=dyn_hair_side_r_01, swing1LimitDegrees=15, swing2LimitDegrees=15, twistLimitDegrees=2}
]
